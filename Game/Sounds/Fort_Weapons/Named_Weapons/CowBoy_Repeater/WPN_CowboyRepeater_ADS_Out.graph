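# Sound cue for the Cowboy Repeater weapon's aim-down-sights exit foley.
# A random node picks with equal weight between two heavy AR ADS-out waves, then feeds a modulator at the root.
# It plays at roughly half volume in the weapon foley sound class, using the shared weapon reload attenuation.
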SoundCue /Game/Sounds/Fort_Weapons/Named_Weapons/CowBoy_Repeater/WPN_CowboyRepeater_ADS_Out
root=Modulator_0
nodes [
 {id=Modulator_0, children [Random_1]}
 {id=Random_1, children [WavePlayer_0,WavePlayer_3], properties {weights [1,1]}}
 {id=WavePlayer_0, properties {soundWaveAssetPtr=/Game/Sounds/Fort_Weapons/ADS/Weapon_HEAVY_AR_ADS_OUT_01}}
 {id=WavePlayer_3, properties {soundWaveAssetPtr=/Game/Sounds/Fort_Weapons/ADS/Weapon_HEAVY_AR_ADS_OUT_02}}
]
properties {
 attenuationSettings=&Fort_Weapon_Reload_Att
 maxDistance=2250
 soundClassObject=&Fort_WeaponFoley
 volumeMultiplier=0.501
}
