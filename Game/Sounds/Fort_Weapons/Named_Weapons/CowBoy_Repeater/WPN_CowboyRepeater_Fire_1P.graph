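# First-person fire sound cue for the Cowboy Repeater weapon.
# A mixer combines a random pick of three gunshot waves with a quieter random shell casing delayed by roughly half a second, then a modulator boosts the overall volume.
# A shotgun sub-punch layer is also wired into the mixer but muted at zero volume.
# Playback uses first-person weapon attenuation, sound class and gunfire concurrency assets.
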SoundCue /Game/Sounds/Fort_Weapons/Named_Weapons/CowBoy_Repeater/WPN_CowboyRepeater_Fire_1P
root=Modulator_17
nodes [
 {id=Delay_0, children [Random_3], properties {delayMax=0.5, delayMin=0.4}}
 {id=Mixer_0, children [Random_4, WavePlayer_10, Delay_0], properties {inputVolume [1,0,0.7]}}
 {id=Modulator_17, children [Mixer_0], properties {pitchMax=1, pitchMin=1, volumeMax=1.3, volumeMin=1.3}}
 {id=Random_3, children [WavePlayer_11, WavePlayer_12, WavePlayer_13], properties {weights [1,1,1]}}
 {id=Random_4, children [WavePlayer_18, WavePlayer_14, WavePlayer_15], properties {weights [1,1,1]}}
 {id=WavePlayer_10, properties {soundWaveAssetPtr=/Game/Sounds/Fort_Weapons/Shotgun/Charge/Sub_Punch_01}}
 {id=WavePlayer_11, properties {soundWaveAssetPtr=&/WPN_CowboyRepeater_Casing_01}}
 {id=WavePlayer_12, properties {soundWaveAssetPtr=&/WPN_CowboyRepeater_Casing_02}}
 {id=WavePlayer_13, properties {soundWaveAssetPtr=&/WPN_CowboyRepeater_Casing_03}}
 {id=WavePlayer_14, properties {soundWaveAssetPtr=&/WPN_CowboyRepeater_Fire_1P_v2_02}}
 {id=WavePlayer_15, properties {soundWaveAssetPtr=&/WPN_CowboyRepeater_Fire_1P_v2_03}}
 {id=WavePlayer_18, properties {soundWaveAssetPtr=&/WPN_CowboyRepeater_Fire_1P_v2_01}}
]
properties {
 attenuationSettings=&Fort_Weapon_Projectile_FirstPerson_Athena_Att
 maxDistance=2097152
 soundClassObject=&Fort_WeaponRanged_Fire_1P_PSM
 volumeMultiplier=1
 attenuationOverrides {distanceAlgorithm=NaturalSound, falloffDistance=4000, attenuationShapeExtents {x=1000}}
 concurrencyOverrides {bLimitToOwner=true, maxCount=3, resolutionRule=StopOldest}
 concurrencySet [&Weapon_Gunfire_FirstPerson_Conc]
}
