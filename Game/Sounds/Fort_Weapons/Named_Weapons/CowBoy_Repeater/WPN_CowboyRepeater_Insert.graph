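# Sound cue for the insert foley of the Cowboy Repeater named weapon, likely heard when rounds are loaded during a reload.
# A Random node picks one of five numbered insert wave variations with equal weight.
# The cue uses the weapon-foley sound class and shared weapon-reload attenuation, with a reduced volume.
# It overrides concurrency and has the debug flag switched on.
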SoundCue /Game/Sounds/Fort_Weapons/Named_Weapons/CowBoy_Repeater/WPN_CowboyRepeater_Insert
root=Random_2
nodes [
 {id=Random_2, children [WavePlayer_8, WavePlayer_9, WavePlayer_10, WavePlayer_11, WavePlayer_12], properties {weights [1,1,1,1,1]}}
 {id=WavePlayer_10, properties {soundWaveAssetPtr=&/WPN_CowboyRepeater_Insert_03}}
 {id=WavePlayer_11, properties {soundWaveAssetPtr=&/WPN_CowboyRepeater_Insert_04}}
 {id=WavePlayer_12, properties {soundWaveAssetPtr=&/WPN_CowboyRepeater_Insert_05}}
 {id=WavePlayer_8, properties {soundWaveAssetPtr=&/WPN_CowboyRepeater_Insert_01}}
 {id=WavePlayer_9, properties {soundWaveAssetPtr=&/WPN_CowboyRepeater_Insert_02}}
]
properties {
 attenuationSettings=&Fort_Weapon_Reload_Att
 bDebug=true
 bOverrideConcurrency=true
 maxDistance=2250
 soundClassObject=&Fort_WeaponFoley
 volumeMultiplier=0.702
}
