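# Sound cue for the lever-action foley of the Cowboy Repeater weapon.
# A random node picks one of three equally weighted lever recordings each time it plays.
# It uses the weapon reload attenuation and weapon foley sound class at reduced volume, with concurrency override and debug flag enabled.
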SoundCue /Game/Sounds/Fort_Weapons/Named_Weapons/CowBoy_Repeater/WPN_CowboyRepeater_Lever
root=Random_1
nodes [
 {id=Random_1, children [WavePlayer_1,WavePlayer_2,WavePlayer_3], properties {weights [1,1,1]}}
 {id=WavePlayer_1, properties {soundWaveAssetPtr=&/WPN_CowboyRepeater_Lever_01}}
 {id=WavePlayer_2, properties {soundWaveAssetPtr=&/WPN_CowboyRepeater_Lever_02}}
 {id=WavePlayer_3, properties {soundWaveAssetPtr=&/WPN_CowboyRepeater_Lever_03}}
]
properties {
 attenuationSettings=&Fort_Weapon_Reload_Att
 bDebug=true
 bOverrideConcurrency=true
 maxDistance=2250
 soundClassObject=&Fort_WeaponFoley
 volumeMultiplier=0.326
}
